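# Niagara effect for the Cube Ninja glider cosmetic, drawing two animated ribbon trails, matching smoke ribbons and drifting rune sprites.
# Trails span socket pairs, while runes sample skeletal mesh sockets, spawn through a movement-threshold module and dim by time of day.
# A system-level FN_Integrator value fades the in-game trails and runes.
# FE_ duplicates spawn only when User.bIsFrontEnd is set, ignore the integrator and use added velocity to stream without glider movement.
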NiagaraSystem /BRCosmetics/Gadgets/Assets/VinderTech_GliderChute/Glider_Cube_Ninja_Male/FX/NS_Glider_Cube_Ninja
dependencies {
 usedDataInterfaces [&NiagaraDataInterfaceColorCurve, &NiagaraDataInterfaceCurve, &NiagaraDataInterfaceArrayFloat]
 usedDynamicInputs []
 usedModules [&SystemState, &FN_Integrator]
}
emitters {
 AnimTrail {
  emitterUpdateScript [
   &EmitterState(Scalability Mode=::NewEnumerator1)
   /Game/Effects/Niagara/Modules/AnimTrail/FN_SpawnAnimTrail/FN_SpawnAnimTrail(Number of Anim Trails=2, SpawnRate=100)
  ]
  particleSpawnScript [
   /Game/Effects/Niagara/Modules/AnimTrail/FN_InitializeAnimTrail/FN_InitializeAnimTrailBoneIndecies(Second Bone or Socket Index=1)
   &FN_InitializeAnimTrailBoneIndecies(AnimTrail Selection=1, First Bone or Socket Index=2, Second Bone or Socket Index=3) {module=FN_InitializeAnimTrailBoneIndecies001}
   /Game/Effects/Niagara/Modules/AnimTrail/FN_InitializeAnimTrail/FN_InitializeAnimTrail(
    AnimTrailWidth=1
    Color=&MultiplyLinearColor(Linear Color=<rgba 1 1 1 1>, Scale Factor=<1 1 1 1>)
    Lifetime=0.175
    Write Material Random=false
   )
  ]
  particleUpdateScript [
   &InheritVelocity(Module.Limit Speed=true, Velocity Limit=20)
   &SolveForcesAndVelocity(Clamp Velocity=false, Limit Acceleration=false)
   /Game/Effects/Niagara/Modules/AnimTrail/FN_UpdateAnimTrail/FN_UpdateAnimTrail(New float=1)
   &ScaleColor(
    Scale Alpha=&Multiply_Float(
     A=&FloatFromCurve(FloatCurve=DataInterface() {propertyValues {Curve {keys [
          {arriveTangent=3.6646, interpMode=RCIM_Cubic, leaveTangent=3.6646, tangentMode=RCTM_User, time=-0.0017, value=0.5086}
          {arriveTangent=-0.0143, interpMode=RCIM_Cubic, leaveTangent=-0.0143, tangentMode=RCTM_User, time=0.1233, value=0.7411}
          {arriveTangent=-1.317, interpMode=RCIM_Cubic, leaveTangent=-1.317, tangentMode=RCTM_User, time=1, value=0}
         ]}}})
     B=&FloatFromCurve(
      FloatCurve=DataInterface() {propertyValues {Curve {keys [<at 0 0 RCIM_Cubic>, {arriveTangent=2.0248, interpMode=RCIM_Cubic, leaveTangent=2.0248, tangentMode=RCTM_User, time=1, value=1}]}}}
      CurveIndex=$System.FN_Integrator.Value
     )
    )
    Scale RGB=&MakeVectorFromLinearColorRGB(Color=&ColorFromCurve(ColorCurve=DataInterface() {propertyValues {
        AlphaCurve {keys [<at 0 1 RCIM_Cubic>]}
        BlueCurve {keys [<at 0.05 1>,<at 0.3 0.9276>,<at 0.5 1>,<at 1 1>]}
        GreenCurve {keys [<at 0.05 0.026>,<at 0.3 0.0918>,<at 0.5 0.2027>,<at 1 0.8323>]}
        RedCurve {keys [<at 0.05 0.789>,<at 0.3 0.1615>,<at 0.5 0.1302>,<at 1 0.1615>]}
       }}))
    ScaleA=true
    ScaleRGB=true
   )
  ]
  rendererClasses [&NiagaraRibbonRendererProperties]
  renderers [{rendererClass=&NiagaraRibbonRendererProperties}]
 }
 AnimTrail_Smoke {
  emitterUpdateScript [
   &EmitterState(Scalability Mode=::NewEnumerator1)
   &FN_SpawnAnimTrail(Number of Anim Trails=2, SpawnRate=100)
  ]
  particleSpawnScript [
   &FN_InitializeAnimTrailBoneIndecies(Second Bone or Socket Index=1)
   &FN_InitializeAnimTrailBoneIndecies(AnimTrail Selection=1, First Bone or Socket Index=2, Second Bone or Socket Index=3) {module=FN_InitializeAnimTrailBoneIndecies001}
   &FN_InitializeAnimTrail(
    AnimTrailWidth=1
    Color=&MultiplyLinearColor(Linear Color=<rgba 1 1 1 0.5>, Scale Factor=<1 1 1 1>)
    Lifetime=0.12
    Write Material Random=false
   )
   &CameraOffset(Camera Offset Amount=&RandomRangeFloat(Minimum=5, Maximum=15))
  ]
  particleUpdateScript [
   &InheritVelocity(Module.Limit Speed=true, Velocity Limit=20)
   &SolveForcesAndVelocity(Clamp Velocity=false, Limit Acceleration=false)
   &FN_UpdateAnimTrail(New float=1)
   &ScaleColor(
    Scale Alpha=&Multiply_Float(
     A=&FloatFromCurve(FloatCurve=DataInterface() {propertyValues {Curve {keys [
          {arriveTangent=3.6646, interpMode=RCIM_Cubic, leaveTangent=3.6646, tangentMode=RCTM_User, time=-0.0017, value=0.5086}
          {arriveTangent=-0.0143, interpMode=RCIM_Cubic, leaveTangent=-0.0143, tangentMode=RCTM_User, time=0.1233, value=0.7411}
          {arriveTangent=-1.317, interpMode=RCIM_Cubic, leaveTangent=-1.317, tangentMode=RCTM_User, time=1, value=0}
         ]}}})
     B=&FloatFromCurve(
      FloatCurve=DataInterface() {propertyValues {Curve {keys [<at 0 0 RCIM_Cubic>, {arriveTangent=2.0248, interpMode=RCIM_Cubic, leaveTangent=2.0248, tangentMode=RCTM_User, time=1, value=1}]}}}
      CurveIndex=$System.FN_Integrator.Value
     )
    )
    Scale RGB=&MakeVectorFromLinearColorRGB(Color=<rgba 1.4879 10 0 1>)
    ScaleA=true
    ScaleRGB=false
   )
  ]
  rendererClasses [&NiagaraRibbonRendererProperties]
  renderers [{rendererClass=&NiagaraRibbonRendererProperties}]
 }
 FE_AnimTrail {
  emitterUpdateScript [
   &EmitterState(Scalability Mode=::NewEnumerator1)
   &FN_SpawnAnimTrail(
    Number of Anim Trails:NiagaraInt32=&MakeCustomIntFromBool(Bool=$User.bIsFrontEnd, True INT=2, False INT=0)
    SpawnRate=40
   )
  ]
  particleSpawnScript [
   &FN_InitializeAnimTrailBoneIndecies(Second Bone or Socket Index=1)
   &FN_InitializeAnimTrailBoneIndecies(AnimTrail Selection=1, First Bone or Socket Index=2, Second Bone or Socket Index=3) {module=FN_InitializeAnimTrailBoneIndecies001}
   &FN_InitializeAnimTrail(
    AnimTrailWidth=1
    Color=&MultiplyLinearColor(Linear Color=<rgba 0.851 0.3125 1 1>, Scale Factor=<1 1 1 1>)
    Lifetime=0.175
    Write Material Random=false
   )
  ]
  particleUpdateScript [
   &AddVelocity(Velocity=<0 -400 0>)
   &InheritVelocity(Module.Limit Speed=true, Velocity Limit=20)
   &SolveForcesAndVelocity(Clamp Velocity=false, Limit Acceleration=false)
   &FN_UpdateAnimTrail(New float=1)
   &ScaleColor(
    Scale Alpha=&Multiply_Float(
     A=&FloatFromCurve(FloatCurve=DataInterface() {propertyValues {Curve {keys [
          {arriveTangent=3.6646, interpMode=RCIM_Cubic, leaveTangent=3.6646, tangentMode=RCTM_User, time=-0.0017, value=0.5086}
          {arriveTangent=-0.0143, interpMode=RCIM_Cubic, leaveTangent=-0.0143, tangentMode=RCTM_User, time=0.1233, value=0.7411}
          {arriveTangent=-1.317, interpMode=RCIM_Cubic, leaveTangent=-1.317, tangentMode=RCTM_User, time=1, value=0}
         ]}}})
     B=1
    )
    Scale RGB=&MakeVectorFromLinearColorRGB(Color=&ColorFromCurve(ColorCurve=DataInterface() {propertyValues {
        AlphaCurve {keys [<at 0 1 RCIM_Cubic>]}
        BlueCurve {keys [<at 0.1 1>,<at 0.3 0.9276>,<at 0.5 1>,<at 1 1>]}
        GreenCurve {keys [<at 0.1 0.026>,<at 0.3 0.0242>,<at 0.5 0.0469>,<at 1 0.8323>]}
        RedCurve {keys [<at 0.1 0.789>,<at 0.3 0.25>,<at 0.5 0.0469>,<at 1 0.1615>]}
       }}))
    ScaleA=true
    ScaleRGB=true
   )
  ]
  rendererClasses [&NiagaraRibbonRendererProperties]
  renderers [{rendererClass=&NiagaraRibbonRendererProperties}]
 }
 FE_AnimTrail_Smoke {
  emitterUpdateScript [
   &EmitterState(Scalability Mode=::NewEnumerator1)
   &FN_SpawnAnimTrail(
    Number of Anim Trails:NiagaraInt32=&MakeCustomIntFromBool(Bool=$User.bIsFrontEnd, True INT=2, False INT=0)
    SpawnRate=40
   )
  ]
  particleSpawnScript [
   &FN_InitializeAnimTrailBoneIndecies(Second Bone or Socket Index=1)
   &FN_InitializeAnimTrailBoneIndecies(AnimTrail Selection=1, First Bone or Socket Index=2, Second Bone or Socket Index=3) {module=FN_InitializeAnimTrailBoneIndecies001}
   &FN_InitializeAnimTrail(
    AnimTrailWidth=1
    Color=&MultiplyLinearColor(Linear Color=<rgba 1 1 1 0.3>, Scale Factor=<1 1 1 1>)
    Lifetime=0.12
    Write Material Random=false
   )
   &CameraOffset(Camera Offset Amount=&RandomRangeFloat(Minimum=5, Maximum=15)) {enabled=false}
  ]
  particleUpdateScript [
   &AddVelocity(Velocity=<0 -500 0>) {module=AddVelocity001}
   &InheritVelocity(Module.Limit Speed=true, Velocity Limit=20)
   &SolveForcesAndVelocity(Clamp Velocity=false, Limit Acceleration=false)
   &FN_UpdateAnimTrail(New float=1)
   &ScaleColor(
    Scale Alpha=&Multiply_Float(
     A=&FloatFromCurve(FloatCurve=DataInterface() {propertyValues {Curve {keys [
          {arriveTangent=3.6646, interpMode=RCIM_Cubic, leaveTangent=3.6646, tangentMode=RCTM_User, time=-0.0017, value=0.5086}
          {arriveTangent=-0.0143, interpMode=RCIM_Cubic, leaveTangent=-0.0143, tangentMode=RCTM_User, time=0.1233, value=0.7411}
          {arriveTangent=-1.317, interpMode=RCIM_Cubic, leaveTangent=-1.317, tangentMode=RCTM_User, time=1, value=0}
         ]}}})
     B=1
    )
    Scale RGB=&MakeVectorFromLinearColorRGB(Color=<rgba 1.4879 10 0 1>)
    ScaleA=true
    ScaleRGB=false
   )
  ]
  rendererClasses [&NiagaraRibbonRendererProperties]
  renderers [{rendererClass=&NiagaraRibbonRendererProperties}]
 }
 FE_Runes {
  emitterUpdateScript [
   &EmitterState(Scalability Mode=::NewEnumerator1)
   &SpawnRate(
    SpawnRate=&MakeCustomFloatFromBool(Bool=$User.bIsFrontEnd, True Float=7, False Float=0)
    Use Spawn Probability=false
   )
  ]
  particleSpawnScript [
   &FN_SetSpawnGroup(Number of Spawn Groups=2)
   &InitializeParticle(
    Color Minimum=<rgba 0 1 0.7167 0.5>
    Color Mode=::NewEnumerator2
    Lifetime Max=1
    Lifetime Min=0.3
    Lifetime Mode=::NewEnumerator1
    Sprite Rotation Mode=::NewEnumerator2
    Sprite Size Mode=::NewEnumerator1
    Uniform Sprite Size Max=20
    Uniform Sprite Size Min=10
    UsePositionOffset=true
    Write Material Random=true
   )
   &SubUVAnimation(End Frame=3, SubUV Animation Mode:ENiagara_SubUVLookupMode=::NewEnumerator2)
   &SkeletalMeshLocation(
    Apply Sampled Orientation to Mesh Particles=true
    Bone / Socket Index=$Particles.SpawnGroup
    Bone Sampling Mode=::NewEnumerator4
    Inherit Velocity=false
    MaskSpawnGroup=false
    Skeletal Mesh=$User.SkeletalMesh001
   )
   &BoxLocation(Box Offset=<0 -40 0>, Box Size=<100 1 1>, MaskSpawnGroup=false, Module.Override Seed=false, Rotate Box=false, Surface Only=false)
   &AddVelocity(Velocity=&RandomRangeVector(Minimum=<0 -300 0>, Maximum=<0 -600 0>))
  ]
  particleUpdateScript [
   &ParticleState()
   &ScaleColor(
    Scale Alpha=&Multiply_Float(
     A=&FloatFromCurve(
      FloatCurve=DataInterface() {propertyValues {Curve {keys [
          {arriveTangent=16.0422, interpMode=RCIM_Cubic, leaveTangent=16.0422, tangentMode=RCTM_User, time=0, value=0}
          <at 0.1721 1.0169 RCIM_Cubic>
          <at 1 0>
         ]}}}
      Scale Curve=0.5
     )
     B=&FNTimeOfDayScalar(Night Scale=0.5, Morning Scale=0.5, Evening Scale=0.5)
    )
    Scale RGB=&MakeVector(
     X=&SelectFloatFromArray(Direct Array Index=$Particles.SpawnGroup, Float Selection Array=DataInterface() {propertyValues {FloatData [0,1]}})
     Y=&FloatFromCurve(FloatCurve=DataInterface() {propertyValues {Curve {keys [<at 0 1>,<at 1 0.5>]}}})
     Z=&RandomRangeFloat(Minimum=2, Maximum=5)
    )
    ScaleA=true
    ScaleRGB=true
   )
   &ScaleSpriteSize(Scale Factor=&Vector2DFromFloat(Value=&Multiply_Float(
      A=&RandomRangeFloat(Minimum=1, Maximum=1, Evaluation Type=::NewEnumerator1)
      B=&FloatFromCurve(FloatCurve=DataInterface() {propertyValues {Curve {keys [
           {arriveTangent=16.0422, interpMode=RCIM_Cubic, leaveTangent=16.0422, tangentMode=RCTM_User, time=0, value=0}
           <at 0.1721 1.0169 RCIM_Cubic>
           <at 1 0>
          ]}}})
     ))) {module=ScaleSpriteSize001}
   &GravityForce(Gravity=&RandomRangeVector(Minimum=<0 0 0>, Maximum=<0 100 0>))
   &Drag(Drag=&RandomRangeFloat(Minimum=2, Maximum=6), Use Linear Drag=true, Use Rotational Drag=false)
   &SolveForcesAndVelocity(Clamp Velocity=false, Limit Acceleration=false)
  ]
  rendererClasses [&NiagaraSpriteRendererProperties]
  renderers [{rendererClass=&NiagaraSpriteRendererProperties}]
 }
 Runes {
  emitterUpdateScript [
   &EmitterState(Scalability Mode=::NewEnumerator1)
   &self:Spawnbymove(MovementThreshold=900, SpawnRate=10) {module=Spawnbymove}
   &SpawnRate(SpawnRate=$Emitter.ThresholdSpawnRate, Use Spawn Probability=false)
  ]
  particleSpawnScript [
   &FN_SetSpawnGroup(Number of Spawn Groups=2)
   &InitializeParticle(
    Color Maximum=<rgba 1 1 1 0.7>
    Color Minimum=<rgba 0.0885 0.5139 1 0.5>
    Color Mode=::NewEnumerator2
    Lifetime Max=1
    Lifetime Min=0.3
    Lifetime Mode=::NewEnumerator1
    Sprite Rotation Mode=::NewEnumerator2
    Sprite Size Mode=::NewEnumerator1
    Uniform Sprite Size Max=25
    Uniform Sprite Size Min=10
    UsePositionOffset=true
    Write Material Random=true
   )
   &SubUVAnimation(End Frame=3, SubUV Animation Mode:ENiagara_SubUVLookupMode=::NewEnumerator2)
   &CameraOffset(Camera Offset Amount=&RandomRangeFloat(Minimum=5, Maximum=20))
   &SkeletalMeshLocation(
    Apply Sampled Orientation to Mesh Particles=true
    Bone / Socket Index=$Particles.SpawnGroup
    Bone Sampling Mode=::NewEnumerator4
    Inherit Velocity=false
    MaskSpawnGroup=false
    Skeletal Mesh=$User.SkeletalMesh001
   )
   &BoxLocation(Box Offset=<0 40 0>, Box Size=<100 1 1>, MaskSpawnGroup=false, Module.Override Seed=false, Rotate Box=false, Surface Only=false)
   &AddVelocity(Velocity=&RandomRangeVector(Minimum=<0 300 0>, Maximum=<0 600 0>))
  ]
  particleUpdateScript [
   &ParticleState()
   &ScaleColor(
    Scale Alpha=&Multiply_Float(
     A=&Multiply_Float(
      A=&FloatFromCurve(FloatCurve=DataInterface() {propertyValues {Curve {keys [
           {arriveTangent=16.0422, interpMode=RCIM_Cubic, leaveTangent=16.0422, tangentMode=RCTM_User, time=0, value=0}
           <at 0.1721 1.0169 RCIM_Cubic>
           <at 1 0>
          ]}}})
      B=&FloatFromCurve(
       FloatCurve=DataInterface() {propertyValues {Curve {keys [<at 0 0 RCIM_Cubic>, {arriveTangent=2.0248, interpMode=RCIM_Cubic, leaveTangent=2.0248, tangentMode=RCTM_User, time=1, value=1}]}}}
       CurveIndex=$System.FN_Integrator.Value
      )
     )
     B=&FNTimeOfDayScalar(Night Scale=0.5, Morning Scale=0.5, Evening Scale=0.5)
    )
    Scale RGB=&MakeVector(
     X=&SelectFloatFromArray(Direct Array Index=$Particles.SpawnGroup, Float Selection Array=DataInterface() {propertyValues {FloatData [0,1]}})
     Y=&Multiply_Float(
      A=&RandomRangeFloat(Minimum=0.5, Maximum=1)
      B=&FloatFromCurve(FloatCurve=DataInterface() {propertyValues {Curve {keys [<at 0 1>,<at 1 0>]}}})
     )
     Z=&RandomRangeFloat(Minimum=2, Maximum=5)
    )
    ScaleA=true
    ScaleRGB=true
   )
   &ScaleSpriteSize(Scale Factor=&Vector2DFromFloat(Value=&Multiply_Float(
      A=&RandomRangeFloat(Minimum=1, Maximum=1, Evaluation Type=::NewEnumerator1)
      B=&FloatFromCurve(FloatCurve=DataInterface() {propertyValues {Curve {keys [
           {arriveTangent=16.0422, interpMode=RCIM_Cubic, leaveTangent=16.0422, tangentMode=RCTM_User, time=0, value=0}
           <at 0.1721 1.0169 RCIM_Cubic>
           <at 1 0>
          ]}}})
     ))) {module=ScaleSpriteSize001}
   &GravityForce(Gravity=&RandomRangeVector(Minimum=<0 0 0>, Maximum=<0 100 0>))
   &Drag(Drag=&RandomRangeFloat(Minimum=2, Maximum=6), Use Linear Drag=true, Use Rotational Drag=false)
   &SolveForcesAndVelocity(Clamp Velocity=false, Limit Acceleration=false)
  ]
  rendererClasses [&NiagaraSpriteRendererProperties]
  renderers [{rendererClass=&NiagaraSpriteRendererProperties}]
 }
}
summary {
 systemName=NS_Glider_Cube_Ninja
 emitters [
  {emitterName=AnimTrail, rendererClasses [&NiagaraRibbonRendererProperties]}
  {emitterName=AnimTrail_Smoke, rendererClasses [&NiagaraRibbonRendererProperties]}
  {emitterName=FE_AnimTrail, rendererClasses [&NiagaraRibbonRendererProperties]}
  {emitterName=FE_AnimTrail_Smoke, rendererClasses [&NiagaraRibbonRendererProperties]}
  {emitterName=Runes, rendererClasses [&NiagaraSpriteRendererProperties]}
  {emitterName=FE_Runes, rendererClasses [&NiagaraSpriteRendererProperties]}
 ]
 userVariables [
  User.bIsFrontEnd {type=NiagaraBool}
  User.ForwardDot {type=NiagaraFloat, defaultValue=0, description=""}
  User.IsFullyDeployed {type=NiagaraBool}
  User.RibbonIDOffset {type=NiagaraInt32}
  User.RightDot {type=NiagaraFloat, defaultValue=0, description=""}
  User.SkeletalMesh {
   type=NiagaraDataInterfaceSkeletalMesh
   defaultValue {
    struct=NiagaraExt_VariableValue_DataInterface
    value {dataInterface=&self:NiagaraDataInterfaceSkeletalMesh_0}
   }
  }
  User.SkeletalMesh001 {
   type=NiagaraDataInterfaceSkeletalMesh
   defaultValue {
    struct=NiagaraExt_VariableValue_DataInterface
    value {dataInterface=&self:NiagaraDataInterfaceSkeletalMesh_1}
   }
  }
 ]
}
